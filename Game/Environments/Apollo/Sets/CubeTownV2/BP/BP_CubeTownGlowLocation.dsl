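; BP_CubeTownGlowLocation is a placeable marker that drives the shared glow material parameters for the CubeTownV2 environment set.
; Its construction script writes the root component's world location into the GlowLocation vector of the MPC_CubetownGlow Material Parameter Collection.
; It also copies its glow scale and radius variables into the GlowScale and GlowRadiusScale scalars.
; BeginPlay, overlap and Tick events are empty, so values are set only at construction.
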(event EventBeginPlay)

(event Collision|EventActorBeginOverlap
  (params (OtherActor :type Actor%20Object%20Reference)))

(event EventTick
  (params (DeltaSeconds :type Float%20%28single-precision%29)))

(fn ConstructionScript
  (Rendering|Material|SetVectorParameterValue "/Game/Environments/Apollo/Sets/CubeTownV2/BP/MPC_CubetownGlow.MPC_CubetownGlow" "GlowLocation" (Math|Conversions|ToLinearColor(Vector) (Transformation|GetWorldLocation (Variables|Default|GetDefaultSceneRoot))) :!class /Script/Engine.KismetMaterialLibrary)
  (Rendering|Material|SetScalarParameterValue "/Game/Environments/Apollo/Sets/CubeTownV2/BP/MPC_CubetownGlow.MPC_CubetownGlow" "GlowScale" (Variables|Default|GetScaleGlowValue) :!class /Script/Engine.KismetMaterialLibrary)
  (Rendering|Material|SetScalarParameterValue "/Game/Environments/Apollo/Sets/CubeTownV2/BP/MPC_CubetownGlow.MPC_CubetownGlow" "GlowRadiusScale" (Variables|Default|GetScale%20Radius) :!class /Script/Engine.KismetMaterialLibrary))
